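# Unlit translucent master material for Engine Volumetrics falloff spheres, exposing color, opacity, distance and falloff parameters.
# Opacity multiplies a clamped pixel-depth ramp by a FalloffExponent power term, whose dot product takes only constants, so it does not vary across the surface.
# A static switch optionally applies DepthFade, noted as replacing UE3's DepthBiasedAlpha.
# Color RGB is multiplied by itself despite its alpha-as-luminosity description, and Density is left unconnected.
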Material /Engine/EngineVolumetrics/FalloffSphere/Materials/M_EV_FalloffSphere_Master_01
graph {
 properties {bUsedWithStaticLighting=true, blendMode=BLEND_Translucent, pixelDepthOffsetMode=PDOM_Legacy, shadingModel=MSM_Unlit, usedShadingModels=Unlit}
 nodes [
  n1 = VectorParameter() {desc="Color (A is luminosity)", parameterName=Color, defaultValue=<rgba 1 1 1 1>}
  Multiply(A=@n1.RGB, B=@n1.RGB)
  ScalarParameter() {defaultValue=0.5, parameterName=Density}
  n2 = Constant() {r=0.0005}
  n3 = ScalarParameter() {defaultValue=1, parameterName=Distance}
  n4 = Clamp(None=Multiply(A=PixelDepth().R, B=Divide(A=@n2, B=@n3)), Max=Constant() {r=1})
  n5 = Constant() {r=0.5}
  n6 = Constant3Vector() {constant=<rgba 0 0 1 0>}
  n7 = Power(Base=Abs(None=DotProduct(A=@n5, B=@n6)), Exp=ScalarParameter() {defaultValue=3, parameterName=FalloffExponent})
  n8 = Clamp(None=Multiply(A=Multiply(A=@n4, B=@n7), B=ScalarParameter() {defaultValue=1, desc=Opacity, parameterName=Opacity}), Max=Constant() {r=1})
  StaticSwitchParameter(
   True=DepthFade(Opacity=@n8) {desc="UE3 used DepthBiasedAlpha expression, and had the Density param hookedup", fadeDistanceDefault=5000}
   False=@n8
  ) {parameterName=UseDepthBiasedAlpha}
 ]
}
